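# Modal widget for sub-game selection that shows on-demand content download progress with a cancel button, built on FortLoadingOnDemandContentModal.
# First activation reveals the hidden layout and plays a staggered Intro; the progress timer starts when Intro finishes, or immediately on later activations.
# Progress events update the bar and percent text after a 0–100 range check, although the value is used as a fraction.
# An Outro animation exists but this graph never plays it.
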WidgetBlueprint /FortniteCoreUI/UI/Login/SubGameSelect/LoadingSubGameContentModal
animations [
 Intro {
  binding_count=4
  length_seconds=0.55
  object=&self:Intro
  sequence {bindings [
    Button_Cancel {
     id=#1
     possessed_class=ConfirmationButton_C
     tracks [
      {
       type=MovieSceneFloatTrack
       display_name=Render Opacity
       sections [{type=MovieSceneFloatSection, channels [None {type=MovieSceneScriptingFloatChannel, keys [[8,0],[11,1]]}]}]
      }
      {
       type=MovieScene2DTransformTrack
       sections [{
         type=MovieScene2DTransformSection
         channels [
          Translation.X {type=MovieSceneScriptingFloatChannel, keys [[8,-50],[11,0]]}
          Translation.Y {type=MovieSceneScriptingFloatChannel, const [0,1]}
          Angle {type=MovieSceneScriptingFloatChannel, const [0,1]}
          Scale.X {type=MovieSceneScriptingFloatChannel, const [1,1]}
          Scale.Y {type=MovieSceneScriptingFloatChannel, const [1,1]}
          Shear.X {type=MovieSceneScriptingFloatChannel, const [0,1]}
          Shear.Y {type=MovieSceneScriptingFloatChannel, const [0,1]}
         ]
        }]
      }
     ]
    }
    CommonTextBlock_Title {
     id=#2
     possessed_class=CommonTextBlock
     tracks [
      {
       type=MovieScene2DTransformTrack
       sections [{
         type=MovieScene2DTransformSection
         channels [
          Translation.X {type=MovieSceneScriptingFloatChannel, keys [[3,-50],[6,0]]}
          Translation.Y {type=MovieSceneScriptingFloatChannel, const [0,1]}
          Angle {type=MovieSceneScriptingFloatChannel, const [0,1]}
          Scale.X {type=MovieSceneScriptingFloatChannel, const [1,1]}
          Scale.Y {type=MovieSceneScriptingFloatChannel, const [1,1]}
          Shear.X {type=MovieSceneScriptingFloatChannel, const [-10,1]}
          Shear.Y {type=MovieSceneScriptingFloatChannel, const [0,1]}
         ]
        }]
      }
      {
       type=MovieSceneFloatTrack
       display_name=Render Opacity
       sections [{type=MovieSceneFloatSection, channels [None {type=MovieSceneScriptingFloatChannel, keys [[3,0],[6,1]]}]}]
      }
     ]
    }
    BodySizeBox {
     id=#3
     possessed_class=SizeBox
     tracks [
      {
       type=MovieScene2DTransformTrack
       sections [{
         type=MovieScene2DTransformSection
         channels [
          Translation.X {type=MovieSceneScriptingFloatChannel, keys [[5,-50],[8,0]]}
          Translation.Y {type=MovieSceneScriptingFloatChannel, const [0,1]}
          Angle {type=MovieSceneScriptingFloatChannel, const [0,1]}
          Scale.X {type=MovieSceneScriptingFloatChannel, const [1,1]}
          Scale.Y {type=MovieSceneScriptingFloatChannel, const [1,1]}
          Shear.X {type=MovieSceneScriptingFloatChannel, const [0,1]}
          Shear.Y {type=MovieSceneScriptingFloatChannel, const [0,1]}
         ]
        }]
      }
      {
       type=MovieSceneFloatTrack
       display_name=Render Opacity
       sections [{type=MovieSceneFloatSection, channels [None {type=MovieSceneScriptingFloatChannel, keys [[5,0],[8,1]]}]}]
      }
     ]
    }
    BackgroundBlurWidget_0 {
     id=#4
     possessed_class=BackgroundBlur
     tracks [{
       type=MovieSceneFloatTrack
       display_name=Render Opacity
       sections [{type=MovieSceneFloatSection, channels [None {type=MovieSceneScriptingFloatChannel, keys [[0,0],[3,1]]}]}]
      }]
    }
   ]}
 }
 Outro {
  binding_count=1
  length_seconds=0.15
  object=&self:Outro
  sequence {bindings [BackgroundBlurWidget_0 {
     id=#5
     possessed_class=BackgroundBlur
     tracks [
      {
       type=MovieSceneFloatTrack
       display_name=Render Opacity
       sections [{type=MovieSceneFloatSection, channels [None {type=MovieSceneScriptingFloatChannel, keys [[0,1],[3,0]]}]}]
      }
      {
       type=MovieScene2DTransformTrack
       sections [{
         type=MovieScene2DTransformSection
         channels [
          Translation.X {type=MovieSceneScriptingFloatChannel, keys [[0,0],[3,-50]]}
          Translation.Y {type=MovieSceneScriptingFloatChannel, const [0,1]}
          Angle {type=MovieSceneScriptingFloatChannel, const [0,1]}
          Scale.X {type=MovieSceneScriptingFloatChannel, const [1,1]}
          Scale.Y {type=MovieSceneScriptingFloatChannel, const [1,1]}
          Shear.X {type=MovieSceneScriptingFloatChannel, const [0,1]}
          Shear.Y {type=MovieSceneScriptingFloatChannel, const [0,1]}
         ]
        }]
      }
     ]
    }]}
 }
]
graphs [EventGraph {nodes [
   Event OnInitialized()
   > SetVisibility(self=Get Modal_VertBox() -> (Modal_VertBox:Vertical Box Object Reference), InVisibility=Hidden)
   n1 = Event SetCancelButtonText!() -> (OutputDelegate, then, InText:"Text (by ref)") {is=title K2Node_Event}
   > n2 = Cast To IconTextButton!(execute, Object=Get Button_Cancel() -> (Button_Cancel:Common Button Legacy Object Reference)) -> (then, CastFailed, AsIcon Text Button:Icon Text Button Object Reference) {is=title K2Node_DynamicCast}
   > Set Text!(execute, self:Icon Text Button Object Reference=@n2, ButtonText:Text=@n1.InText) -> (then) {is=title K2Node_CallFunction}
   Event On Activated()
   > n3 = Branch(Condition=Get HasPlayedAnim() -> (HasPlayedAnim:bool))
   > StartProgressUpdateTimer(self:Fort Loading On Demand Content Modal Object Reference)
   SetVisibility(execute=@n3.else, self=Get Modal_VertBox(), InVisibility=Visible)
   > Set HasPlayedAnim(HasPlayedAnim:bool=true) -> (Output_Get:bool)
   > PlayAnimation(InAnimation=Get Intro())
   n4 = Event OnUpdatedProgress!() -> (OutputDelegate, then, Percent:float) {is=title K2Node_Event}
   > Branch(Condition=AND Boolean(
     A="float >= float"(A=@n4.Percent, B=0) {is=title K2Node_CallFunction}
     B="float <= float"!(A:double=@n4.Percent, B:double=100) -> (ReturnValue:bool) {is=title K2Node_CallFunction}
    ))
   > SetPercent!(
    execute
    self:Progress Bar Object Reference=Get DownloadingProgressBar() -> (DownloadingProgressBar:Progress Bar Object Reference)
    InPercent:float=@n4.Percent
   ) -> (then)
   > "SetText (Text)"(
    self=Get Text_Progress() -> (Text_Progress:Common Text Object Reference)
    InText=As Percent!(
     Value:float=@n4.Percent
     RoundingMode:ERoundingMode Enum
     bAlwaysSign:bool=false
     bUseGrouping:bool=false
     MinimumIntegralDigits:int=1
     MaximumIntegralDigits:int=324
     MinimumFractionalDigits:int=0
     MaximumFractionalDigits:int=0
    ) -> (ReturnValue:Text)
   )
   n5 = Event OnAnimationFinished!() -> (OutputDelegate, then, Animation:Widget Animation Object Reference) {is=title K2Node_Event}
   > Branch(Condition="Equal (Object)"(A=@n5.Animation, B=Get Intro()) {is=title K2Node_CallFunction})
   > StartProgressUpdateTimer()
  ]}]
variables [HasPlayedAnim {type=boolean}]
widgets {
 info {
  inheritedWidgetCount=1
  parentClass=/Script/FortniteUI.FortLoadingOnDemandContentModal
  rootWidgetClass=&BackgroundBlur
  widgetCount=14
 }
 widgets [
  BackgroundBlurWidget_0 {widgetClassPath=&BackgroundBlur, widget=BackgroundBlurWidget_0, properties {blurStrength=15}}
  Overlay_1 {
   widgetClassPath=&Overlay
   parent=BackgroundBlurWidget_0
   widget=Overlay_1
   slot {class=BackgroundBlurSlot, properties {padding=<ltrb 0 0 0 0>}}
  }
  Lightbox {
   widgetClassPath=/Game/UI/Foundation/Lightbox/Lightbox
   bIsVariable=true
   parent=Overlay_1
   widget=Lightbox
   slot {class=OverlaySlot, properties {horizontalAlignment=HAlign_Fill, verticalAlignment=VAlign_Fill}}
  }
  Modal_VertBox {
   widgetClassPath=&VerticalBox
   bIsVariable=true
   parent=Overlay_1
   widget=Modal_VertBox
   slot {class=OverlaySlot, properties {horizontalAlignment=HAlign_Center, verticalAlignment=VAlign_Center}}
  }
  CommonTextBlock_Title {
   widgetClassPath=&CommonTextBlock
   bIsVariable=true
   parent=Modal_VertBox
   widget=CommonTextBlock_Title
   properties {
    clipping=ClipToBounds
    scrollStyle=&TextScrollStyle-Base_C
    text=Downloading Content....
    textTransformPolicy=ToUpper
    font {
     bForceMonospaced=false
     bMaterialIsStencil=false
     fontMaterial=null
     fontObject=&Burbank
     letterSpacing=0
     monospacedWidth=1
     size=45
     skewAmount=0
     typefaceFontName=BurbankRegular-Black
     outlineSettings {
      bApplyOutlineToDropShadows=false
      bMiteredCorners=false
      bSeparateFillAlpha=false
      outlineMaterial=null
      outlineSize=0
      outlineColor=<rgba 0 0 0 1>
     }
    }
    renderTransform {angle=0, scale=<1 1>, shear=<-10 0>, translation=<0 0>}
    shadowOffset=<0 0>
   }
   slot {class=VerticalBoxSlot, properties {padding=<ltrb 25 0 0 5>}}
  }
  BodySizeBox {
   widgetClassPath=&SizeBox
   parent=Modal_VertBox
   widget=BodySizeBox
   properties {bOverride_MinDesiredHeight=true, bOverride_WidthOverride=true, minDesiredHeight=140, widthOverride=900}
   slot {class=VerticalBoxSlot}
  }
  CommonBorder {
   widgetClassPath=&CommonBorder
   parent=BodySizeBox
   widget=CommonBorder
   properties {
    visibility=Visible
    background {
     bIsDynamicallyLoaded=false
     drawAs=Box
     imageType=NoImage
     mirroring=NoMirror
     resourceName=None
     resourceObject=/FortniteCoreUI/UI/Login/SubGameSelect/Materials/M_UI_SkewBox_LoadingSubGameContentBackground
     tiling=NoTile
     imageSize=<32 32>
     margin=<ltrb 0 0 0 0>
     outlineSettings {
      bUseBrushTransparency=false
      roundingType=HalfHeightRadius
      width=0
      color {colorUseRule=UseColor_Specified, specifiedColor=<rgba 0 0 0 0>}
      cornerRadii=<0 0 0 0>
     }
     tintColor {colorUseRule=UseColor_Specified, specifiedColor=<rgba 1 1 1 1>}
     uVRegion {bIsValid=false, max=<0 0>, min=<0 0>}
    }
    brushColor=<rgba 0.0003 0.0395 0.3372 1>
    padding=<ltrb 0 0 0 0>
   }
   slot {class=SizeBoxSlot}
  }
  VerticalBox {
   widgetClassPath=&VerticalBox
   parent=CommonBorder
   widget=VerticalBox
   properties {navigation=&self:WidgetTree.VerticalBox.WidgetNavigation_0}
   slot {class=BorderSlot, properties {padding=<ltrb 0 0 0 0>}}
  }
  Text_Content {
   widgetClassPath=&CommonTextBlock
   bIsVariable=true
   parent=VerticalBox
   widget=Text_Content
   properties {
    autoWrapText=true
    mobileFontSizeMultiplier=1.4
    text=Please be patient until content has been downloaded
    colorAndOpacity {colorUseRule=UseColor_Specified, specifiedColor=<rgba 0.1946 0.5972 0.9473 1>}
    font {
     bForceMonospaced=false
     bMaterialIsStencil=false
     fontMaterial=null
     fontObject=&BurbankSmall
     letterSpacing=0
     monospacedWidth=1
     size=21
     skewAmount=0
     typefaceFontName=Bold
     outlineSettings {
      bApplyOutlineToDropShadows=false
      bMiteredCorners=false
      bSeparateFillAlpha=false
      outlineMaterial=null
      outlineSize=0
      outlineColor=<rgba 0 0 0 1>
     }
    }
    renderTransform {angle=0, scale=<1 1>, shear=<-10 0>, translation=<0 0>}
    shadowOffset=<0 0>
   }
   slot {class=VerticalBoxSlot, properties {padding=<ltrb 25 20 0 0>}}
  }
  ProgressBarOverlay {
   widgetClassPath=&Overlay
   parent=VerticalBox
   widget=ProgressBarOverlay
   slot {class=VerticalBoxSlot, properties {padding=<ltrb 25 25 35 0>}}
  }
  ProgressSizeBox {
   widgetClassPath=&SizeBox
   parent=ProgressBarOverlay
   widget=ProgressSizeBox
   properties {bOverride_WidthOverride=true, widthOverride=840}
   slot {class=OverlaySlot, properties {horizontalAlignment=HAlign_Fill, verticalAlignment=VAlign_Fill}}
  }
  DownloadingProgressBar {
   widgetClassPath=/Script/UMG.ProgressBar
   bIsVariable=true
   parent=ProgressSizeBox
   widget=DownloadingProgressBar
   properties {
    percent=0.5
    fillColorAndOpacity=<rgba 0 0.5 1 1>
    widgetStyle {
     enableFillAnimation=false
     backgroundImage {
      bIsDynamicallyLoaded=false
      drawAs=Image
      imageType=FullColor
      mirroring=NoMirror
      resourceName=../../../Engine/Content/Slate/Common/ProgressBar_Background.png
      resourceObject=/Game/Global/Textures/Default/Blanks/T_Blank_LegalWhite_sRGBColor
      tiling=NoTile
      imageSize=<16 16>
      margin=<ltrb 0.4167 0.4167 0.4167 0.4167>
      outlineSettings {
       bUseBrushTransparency=false
       roundingType=HalfHeightRadius
       width=0
       color {colorUseRule=UseColor_Specified, specifiedColor=<rgba 0 0 0 0>}
       cornerRadii=<0 0 0 0>
      }
      tintColor {colorUseRule=UseColor_Specified, specifiedColor=<rgba 0.0033 0.0097 0.1022 1>}
      uVRegion {bIsValid=false, max=<0 0>, min=<0 0>}
     }
     fillImage {
      bIsDynamicallyLoaded=false
      drawAs=Image
      imageType=FullColor
      mirroring=NoMirror
      resourceName=../../../Engine/Content/Slate/Common/ProgressBar_Fill.png
      resourceObject=&T_Blank_LegalWhite_sRGBColor
      tiling=NoTile
      imageSize=<16 16>
      margin=<ltrb 0.4167 0.4167 0.4167 0.4167>
      outlineSettings {
       bUseBrushTransparency=false
       roundingType=HalfHeightRadius
       width=0
       color {colorUseRule=UseColor_Specified, specifiedColor=<rgba 0 0 0 0>}
       cornerRadii=<0 0 0 0>
      }
      tintColor {colorUseRule=UseColor_Specified, specifiedColor=<rgba 1 1 0.9801 1>}
      uVRegion {bIsValid=false, max=<0 0>, min=<0 0>}
     }
     marqueeImage {
      bIsDynamicallyLoaded=false
      drawAs=Image
      imageType=FullColor
      mirroring=NoMirror
      resourceName=../../../Engine/Content/Slate/Common/ProgressBar_Marquee.png
      resourceObject=null
      tiling=Horizontal
      imageSize=<20 30>
      margin=<ltrb 0 0 0 0>
      outlineSettings {
       bUseBrushTransparency=false
       roundingType=HalfHeightRadius
       width=0
       color {colorUseRule=UseColor_Specified, specifiedColor=<rgba 0 0 0 0>}
       cornerRadii=<0 0 0 0>
      }
      tintColor {colorUseRule=UseColor_Specified, specifiedColor=<rgba 1 1 1 1>}
      uVRegion {bIsValid=false, max=<0 0>, min=<0 0>}
     }
    }
   }
   slot {class=SizeBoxSlot}
  }
  Text_Progress {
   widgetClassPath=&CommonTextBlock
   bIsVariable=true
   parent=ProgressBarOverlay
   widget=Text_Progress
   properties {
    autoWrapText=true
    justification=Center
    mobileFontSizeMultiplier=1.4
    text=100%
    font {
     bForceMonospaced=false
     bMaterialIsStencil=false
     fontMaterial=null
     fontObject=&BurbankSmall
     letterSpacing=0
     monospacedWidth=1
     size=21
     skewAmount=0
     typefaceFontName=Bold
     outlineSettings {
      bApplyOutlineToDropShadows=false
      bMiteredCorners=false
      bSeparateFillAlpha=false
      outlineMaterial=null
      outlineSize=0
      outlineColor=<rgba 0 0 0 1>
     }
    }
    shadowOffset=<0 0>
   }
   slot {class=OverlaySlot, properties {horizontalAlignment=HAlign_Center, verticalAlignment=VAlign_Center}}
  }
  Button_Cancel {
   widgetClassPath=/Game/UI/Foundation/Modals/ConfirmationWindow/ConfirmationButton
   bInherited=true
   bIsVariable=true
   parent=Modal_VertBox
   widget=Button_Cancel
   properties {
    buttonText=Back
    minHeight=65
    minWidth=200
    onButtonBaseClicked=(&self:WidgetTree.Button_Cancel.HandleOnButtonClicked)
    onButtonBaseDoubleClicked=(&self:WidgetTree.Button_Cancel.HandleOnButtonDoubleClicked)
    onButtonBaseHovered=(&self:WidgetTree.Button_Cancel.HandleOnButtonHovered)
    onButtonBaseUnhovered=(&self:WidgetTree.Button_Cancel.HandleOnButtonUnhovered)
    onSelectedChangedBase=(&self:WidgetTree.Button_Cancel.HandleOnSelectedChanged)
    style=/FortUILibrary/Foundation/Buttons/ButtonStyles/Athena/Athena_ButtonStyle_AngledDarkBlueMenuButton.Athena_ButtonStyle_AngledDarkBlueMenuButton_C
    triggeringInputAction {dataTable=&InputActionDataTable, rowName=Cancel}
   }
   slot {class=VerticalBoxSlot, properties {horizontalAlignment=HAlign_Right, padding=<ltrb 0 20 50 0>}}
  }
 ]
}
