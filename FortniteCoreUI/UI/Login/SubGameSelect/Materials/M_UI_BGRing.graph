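# M_UI_BGRing is an additive UI material that draws an expanding, fading ring pulse, stored under the SubGameSelect login UI.
# Time scaled by PulseSpeed wraps into a repeating 0-1 phase, shaped by FalloffPower.
# That phase animates the radii of two exponential radial gradients on normalized UI UVs, each with its own density.
# The outer gradient multiplied by the inverted inner one forms the ring, which fades out linearly with the shaped phase.
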Material /FortniteCoreUI/UI/Login/SubGameSelect/Materials/M_UI_BGRing
graph {
 properties {bUsedWithStaticMesh=false, blendMode=BLEND_Additive, materialDomain=MD_UI, pixelDepthOffsetMode=PDOM_Legacy, usedShadingModels=Default Lit}
 nodes [
  n1 = MaterialFunctionCall() {materialFunction=&GetUserInterfaceUV}
  n2 = Time()
  n3 = ScalarParameter() {defaultValue=0.2, parameterName=PulseSpeed}
  n4 = Power(Base=Fmod(A=Multiply(A=@n2, B=@n3), B=Constant() {r=1}), Exp=ScalarParameter() {defaultValue=2, parameterName=FalloffPower})
  n5 = LinearInterpolate(
   A=ScalarParameter() {group=OuterRing, parameterName=OuterRingStart}
   B=ScalarParameter() {defaultValue=1, group=OuterRing, parameterName=OuterRingStop}
   Alpha=@n4
  )
  n6 = ScalarParameter() {defaultValue=2.3, group=OuterRing, parameterName=OuterRingDensity}
  n7 = MaterialFunctionCall(
   UVs=@n1.Normalized UV
   Radius=LinearInterpolate(
    A=ScalarParameter() {group=InnerRing, parameterName=InnerRingStart}
    B=ScalarParameter() {defaultValue=2, group=InnerRing, parameterName=InnerRingStop}
    Alpha=@n4
   )
   Density=ScalarParameter() {defaultValue=2.3, group=InnerRing, parameterName=InnerRingDensity}
  ) {materialFunction=&RadialGradientExponential}
  Multiply(
   A=Multiply(
    A=MaterialFunctionCall(UVs=@n1.Normalized UV, Radius=@n5, Density=@n6) {materialFunction=&RadialGradientExponential}.RadialGradientExponential
    B=OneMinus(None=@n7.RadialGradientExponential)
   )
   B=LinearInterpolate(Alpha=@n4) {constA=1, constB=0}
  )
 ]
}
